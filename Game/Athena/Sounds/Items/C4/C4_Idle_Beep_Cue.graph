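# Sound cue for the idle beep of an armed C4 item in Athena.
# A single wave player plays the armed idle beep sample at reduced volume in the Fort Gadgets Items sound class.
# It uses a dedicated attenuation asset for distance falloff and overrides concurrency so new instances are prevented once the limit is reached.
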SoundCue /Game/Athena/Sounds/Items/C4/C4_Idle_Beep_Cue
root=WavePlayer_3
nodes [{id=WavePlayer_3, properties {soundWaveAssetPtr=&/C4_Armed_Idle_Beep_01}}]
properties {
 attenuationSettings=&/C4_Idle_Beep_Att
 bOverrideConcurrency=true
 maxDistance=1500
 soundClassObject=&Fort_Gadgets_Items
 volumeMultiplier=0.6
 concurrencyOverrides {resolutionRule=PreventNew}
}
