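# Niagara system for the C4 consumable's blinking flare: a short team-colored glint with an idle beep.
# A single Flare emitter burst-spawns short-lived particles tinted by User.TeamColor, nudged by camera offset, each playing the beep cue on spawn.
# Curves swell and shrink sprite size plus a separate SizeFlareSharp attribute, while boosted color fades in and out.
# Two sprite renderers and a light renderer draw it, with the team-colored light raised slightly above the spawn point.
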NiagaraSystem /Game/Athena/Items/Consumables/C4/NS_C4_Flare
dependencies {
 usedDataInterfaces [&NiagaraDataInterfaceAudioPlayer, &NiagaraDataInterfaceCurve, &NiagaraDataInterfaceVectorCurve]
 usedDynamicInputs []
 usedModules [&SystemState]
}
emitters {Flare {
  emitterUpdateScript [
   &EmitterState()
   &SpawnBurst_Instantaneous(Use Spawn Probability=false, UseLoopCountLimit=false)
  ]
  particleSpawnScript [
   &InitializeParticle(
    Color=$User.TeamColor
    Lifetime=0.35
    Sprite Size Mode=::NewEnumerator3
    Uniform Sprite Size=50
    UsePositionOffset=false
    Write Material Random=true
   )
   /SetVariables(Particles.SizeFlareSharp:Vector2f=<250 60>) {isSetParameters=true}
   &CameraOffset(Camera Offset Amount=8)
   /Niagara/Modules/Audio/PlayAudio(Audio Player=DataInterface() {propertyValues {
      Attenuation=null
      Concurrency=null
      MaxPlaysPerTick=10
      SoundToPlay=/Game/Athena/Sounds/Items/C4/C4_Idle_Beep_Cue
      bAllowLoopingOneShotSounds=false
      bLimitPlaysPerTick=true
      bOnlyActiveDuringGameplay=false
      bStopWhenComponentIsDestroyed=true
      ConfigurationUserParameter {parameter {name=None, typeDefHandle {registeredTypeIndex=1}, varData []}}
      ParameterNames []
     }})
  ]
  particleUpdateScript [
   &ParticleState()
   &ScaleSpriteSize(Uniform Curve Sprite Scale=DataInterface() {propertyValues {Curve {keys [
        {arriveTangent=6.3733, interpMode=RCIM_Cubic, leaveTangent=6.3733, leaveTangentWeight=0.3504, tangentMode=RCTM_User, time=0, value=0.4}
        {arriveTangentWeight=0.3504, interpMode=RCIM_Cubic, leaveTangentWeight=0.4672, tangentMode=RCTM_User, time=0.2, value=1}
        {arriveTangentWeight=0.4672, interpMode=RCIM_Cubic, leaveTangentWeight=0.3504, tangentMode=RCTM_User, time=0.7, value=1}
        {arriveTangent=-4.713, arriveTangentWeight=0.1, interpMode=RCIM_Cubic, leaveTangent=-4.713, tangentMode=RCTM_User, time=1, value=0.4}
       ]}}})
   /SetVariables(Particles.SizeFlareSharp:Vector2f=/Niagara/DynamicInputs/ScaleByCurve/ScaleVector2DByCurve(
     InitialValue:Vector2f=$Particles.Initial.SizeFlareSharp
     ScaleCurve:NiagaraDataInterfaceVectorCurve=DataInterface() {propertyValues {
       XCurve {keys [
         {arriveTangent=7.2511, interpMode=RCIM_Cubic, leaveTangent=7.2511, tangentMode=RCTM_User, time=0, value=0.4}
         {interpMode=RCIM_Cubic, tangentMode=RCTM_User, time=0.2, value=1}
         {interpMode=RCIM_Cubic, tangentMode=RCTM_User, time=0.6, value=1}
         {arriveTangentWeight=0.1356, interpMode=RCIM_Cubic, tangentMode=RCTM_User, time=1, value=0.4}
        ]}
       YCurve {keys [
         {arriveTangent=7.2511, interpMode=RCIM_Cubic, leaveTangent=7.2511, tangentMode=RCTM_User, time=0, value=0.4}
         {interpMode=RCIM_Cubic, tangentMode=RCTM_User, time=0.2, value=1}
         {interpMode=RCIM_Cubic, tangentMode=RCTM_User, time=0.6, value=1}
         {arriveTangentWeight=0.1356, interpMode=RCIM_Cubic, tangentMode=RCTM_User, time=1, value=0.4}
        ]}
       ZCurve {keys [
         {arriveTangent=7.2511, interpMode=RCIM_Cubic, leaveTangent=7.2511, tangentMode=RCTM_User, time=0, value=0.4}
         {interpMode=RCIM_Cubic, tangentMode=RCTM_User, time=0.2, value=1}
         {interpMode=RCIM_Cubic, tangentMode=RCTM_User, time=0.6, value=1}
         {arriveTangentWeight=0.1356, interpMode=RCIM_Cubic, tangentMode=RCTM_User, time=1, value=0.4}
        ]}
      }}
     CurveIndex:NiagaraFloat=$Particles.NormalizedAge
    )) {isSetParameters=true}
   &ScaleColor(
    Scale Alpha=&FloatFromCurve(FloatCurve=DataInterface() {propertyValues {Curve {keys [
         {arriveTangent=9.4758, interpMode=RCIM_Cubic, leaveTangent=9.4758, tangentMode=RCTM_User, time=0, value=0}
         {interpMode=RCIM_Cubic, tangentMode=RCTM_User, time=0.2, value=1}
         {interpMode=RCIM_Cubic, tangentMode=RCTM_User, time=0.6, value=1}
         {interpMode=RCIM_Cubic, tangentMode=RCTM_User, time=1, value=0}
        ]}}})
    Scale RGB=&VectorFromFloat(Value=4)
    ScaleA=true
    ScaleRGB=true
   )
   &FN_Light_Attributes(
    Color=$User.TeamColor
    Create Color Attribute=true
    Create Exponent Attribute=true
    Create Position Attribute=true
    Create Radius Attribute=true
    Exponent=5
    Position=&AddVectorToPosition(Position=$Particles.Initial.Position, Vector=<0 0 15>)
    Radius=120
    Scale Color=5
   )
  ]
  rendererClasses [&NiagaraSpriteRendererProperties, &NiagaraLightRendererProperties]
  renderers [
   {rendererClass=&NiagaraSpriteRendererProperties}
   {rendererClass=&NiagaraSpriteRendererProperties, rendererIndex=1}
   {rendererClass=&NiagaraLightRendererProperties, rendererIndex=2}
  ]
 }}
summary {
 systemName=NS_C4_Flare
 emitters [{emitterName=Flare, rendererClasses [&NiagaraSpriteRendererProperties, &NiagaraLightRendererProperties]}]
 userVariables [User.TeamColor {type=LinearColor, description="", defaultValue {struct=LinearColor, value=<rgba 0 0.2815 1 1>}}]
}
